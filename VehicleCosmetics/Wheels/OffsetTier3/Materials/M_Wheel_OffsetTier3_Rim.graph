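# Vehicle cosmetic rim material for the OffsetTier3 wheel: dark metal with a normal map and animated PaintColor glow.
# UV-distance math builds a central disc and surrounding ring without a mask texture.
# A slow time-driven sine blends the emissive from solid-colour ring to disc with scrolling radial-gradient pulses, passed via named reroute.
# The material attributes then chain through the shared vehicle damage and de-rez transition functions.
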Material /VehicleCosmetics/Wheels/OffsetTier3/Materials/M_Wheel_OffsetTier3_Rim
graph {
 properties {
  bAutomaticallySetUsageInEditor=false
  bUseMaterialAttributes=true
  bUsedWithSkeletalMesh=true
  previewMesh=/VehicleCosmetics/Wheels/OffsetTier3/meshes/SK_Wheel_OffsetTier3_
  usedShadingModels=Default Lit
 }
 nodes [
  n1 = TextureSampleParameter2D() {
   constCoordinate=1
   parameterName=Normal
   samplerType=SAMPLERTYPE_Normal
   texture=/VehicleCosmetics/Wheels/OffsetTier3/Textures/T_Wheel_OffsetTier3_Rim_N
  }
  n2 = SetMaterialAttributes!(
   MaterialAttributes
   Base Color=VectorParameter() {parameterName=BC, defaultValue=<rgba 0.0573 0.0573 0.0573 1>}.RGB
   Emissive Color=NamedRerouteUsage()
   Roughness=Constant() {r=0.3}
   Normal=@n1.RGB
   Metallic=ScalarParameter() {defaultValue=0.9, parameterName=Metallic}
  ) {attributeSetTypes [
    #1
    #2
    #3
    #4
    #5
   ]}
  MaterialFunctionCall(In=MaterialFunctionCall(Material=@n2) {materialFunction=&MF_BaseVehicle_Damage}.Result) {materialFunction=&MF_VehicleDeRez_Transition}
  n3 = VectorParameter() {group=Color, parameterName=PaintColor, defaultValue=<rgba 0 0.5029 0 1>}
  n4 = Constant() {r=0.5}
  n5 = Constant2Vector() {g=0.5, r=0.5}
  n6 = Multiply(
   A=ComponentMask(None=MaterialFunctionCall!(Radius=@n4, CenterPosition=@n5) {materialFunction=&/MF_RadialGradient}.RadialGradient) {g=true, r=true}
   B=Constant2Vector() {g=0.5, r=0.05}
  )
  n7 = Sine(None=Panner(Coordinate=ComponentMask(None=@n6) {g=true, r=true}, Time=Time()) {speedY=1})
  n8 = Multiply(A=Clamp(None=ComponentMask(None=@n7) {g=true}) {minDefault=0.02}, B=@n3.RGB)
  n9 = ComponentMask(None=TextureCoordinate() {uTiling=0.0001, vTiling=0}) {g=true, r=true}
  n10 = Sine(None=ComponentMask(None=Panner(Coordinate=@n9) {speedX=0.05}) {r=true}) {period=0.5}
  n11 = OneMinus(None=Clamp(None=ConstantBiasScale(None=@n10) {bias=-0.1, scale=2}) {minDefault=0.02})
  n12 = Add(A=TextureCoordinate(), B=Constant() {r=-0.5})
  n13 = OneMinus(None=DotProduct(A=@n12, B=@n12))
  n14 = Constant() {r=0.96}
  n15 = Clamp(None=Multiply(A=Subtract(A=@n13, B=@n14), B=Constant() {r=300}))
  n16 = Add(A=TextureCoordinate(), B=Constant() {r=-0.5})
  n17 = Subtract(A=OneMinus(None=DotProduct(A=@n16, B=@n16)), B=Constant() {r=0.93})
  n18 = Constant() {r=300}
  n19 = Multiply(A=OneMinus(None=@n15), B=Clamp(None=Multiply(A=@n17, B=@n18)))
  NamedRerouteDeclaration(None=Multiply(A=LinearInterpolate(A=@n3.RGB, B=@n8, Alpha=@n11), B=LinearInterpolate(A=@n19, B=@n15, Alpha=@n11))) {name=Emissive, nodeColor=<rgba 1 0 0.9126 1>}
 ]
}
