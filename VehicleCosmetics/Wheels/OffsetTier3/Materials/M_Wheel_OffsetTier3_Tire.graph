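# Animated skeletal-mesh tire material for the OffsetTier3 cosmetic wheel: near-black rubber with glowing, moving elements in PaintColor.
# Pivot Painter data offsets and scales each element through World Position Offset along its X axis.
# A slow time sine blends random per-element pulses with a panning UV1 wave, inverted by custom alpha group, driving both motion and Fresnel-boosted emissive.
# Attributes then pass through shared vehicle damage and de-rez transition functions.
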Material /VehicleCosmetics/Wheels/OffsetTier3/Materials/M_Wheel_OffsetTier3_Tire
graph {
 properties {bAutomaticallySetUsageInEditor=false, bUseMaterialAttributes=true, bUsedWithSkeletalMesh=true, usedShadingModels=Default Lit}
 nodes [
  n1 = SetMaterialAttributes!(
   MaterialAttributes
   Base Color=VectorParameter() {parameterName=BC, defaultValue=<rgba 0.0052 0.0052 0.0052 1>}.RGB
   Emissive Color=NamedRerouteUsage()
   World Position Offset=NamedRerouteUsage()
   Normal=NamedRerouteUsage()
  ) {attributeSetTypes [
    #1
    #2
    #3
    #4
   ]}
  MaterialFunctionCall(In=MaterialFunctionCall(Material=@n1) {materialFunction=&MF_BaseVehicle_Damage}.Result) {materialFunction=&MF_VehicleDeRez_Transition}
  n2 = MaterialFunctionCall() {materialFunction=/Engine/Functions/Engine_MaterialFunctions02/PivotPainter/PivotPainter_PerObjectData}
  n3 = ScalarParameter() {defaultValue=3.2, parameterName=WheelScaleMultiplyer}
  Multiply(A=@n2.Random Value Per Element, B=@n3) {constB=3.2}
  NamedRerouteDeclaration(None=TextureSample() {samplerType=SAMPLERTYPE_Normal, texture=/VehicleCosmetics/Wheels/OffsetTier3/Textures/T_Wheel_OffsetTier3_N}.RGB) {name=N, nodeColor=<rgba 0.3359 0 1 1>}
  n4 = Panner(Coordinate=TextureCoordinate() {coordinateIndex=1, uTiling=0.27, vTiling=0}) {speedX=0.2}
  n5 = ConstantBiasScale(None=Sine(None=ComponentMask(None=@n4) {r=true}) {period=0.5}) {bias=-0.1, scale=6}
  n6 = Clamp(None=@n5)
  n7 = Clamp(None=OneMinus(None=@n5))
  n8 = Clamp(None=Ceil(None=ConstantBiasScale(None=@n2.Custom Alpha Values) {bias=-0.25, scale=2}))
  n9 = LinearInterpolate(A=@n6, B=@n7, Alpha=@n8)
  n10 = Multiply(A=Time(), B=Constant() {r=0.05})
  n11 = Clamp(None=Multiply(A=Sine(None=@n10) {period=0.5}, B=Constant() {r=5}))
  n12 = Panner(Coordinate=@n2.Random Value Per Element) {speedX=0.1}
  n13 = Multiply(A=@n2.Random Value Per Element, B=Sine(None=ComponentMask(None=@n12) {r=true}) {period=0.1})
  n14 = Clamp(None=Multiply(A=ScalarParameter() {defaultValue=2, parameterName=AnimMultiply}, B=LinearInterpolate(A=@n13, B=@n9, Alpha=@n11))) {maxDefault=3, minDefault=0.01}
  n15 = Subtract(A=WorldPosition().XYZ, B=Multiply(A=@n2.Pivot Position, B=@n3) {constB=3.2})
  n16 = ScalarParameter() {defaultValue=0.03, parameterName=Scale}
  n17 = Add(
   A=Multiply(A=Multiply(A=@n2.X-Axis Vector, B=@n3) {constB=3.2}, B=@n14)
   B=LinearInterpolate(A=Multiply(A=@n15, B=@n16), B=Constant(), Alpha=@n14)
  )
  NamedRerouteDeclaration(None=@n17) {name=WPO, nodeColor=<rgba 0 0.0991 0.8125 1>}
  n18 = NamedRerouteUsage()
  n19 = ScalarParameter() {defaultValue=10, parameterName=FresnelExponent}
  n20 = OneMinus(None=@n8)
  n21 = LinearInterpolate(A=@n13, B=Max(A=Multiply(A=@n6, B=@n20), B=Multiply(A=@n7, B=@n8)), Alpha=@n11)
  n22 = Multiply(
   A=Multiply(
    A=ScalarParameter() {defaultValue=2, parameterName=Fresnel_Emissive}
    B=MaterialFunctionCall!(Normal=@n18, Exponent=@n19) {materialFunction=/VehicleCosmetics/Wheels/Cola_Holographic/Materials/MF_Fresnel_Classic}.Result
   )
   B=@n21
  )
  n23 = LinearInterpolate(A=Max(A=@n22, B=@n21), B=@n22, Alpha=@n11)
  NamedRerouteDeclaration(None=Multiply(A=VectorParameter() {group=Color, parameterName=PaintColor, defaultValue=<rgba 0 0.5029 0 1>}.RGB, B=Clamp(None=@n23) {maxDefault=3})) {name=Emissive, nodeColor=<rgba 0.7344 0 1 1>}
  OneMinus(None=@n17)
 ]
}
